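# Translucent UI material for the game mode background in the frontend Activity Browser.
# Color blends two parameters through an aspect-corrected radial gradient.
# With the Transperant switch on, alpha averages two rotated, differently tiled triangle-grid samples, fading to solid through a second radial gradient; otherwise it uses Opacity, which also multiplies both results.
# A panning sine mask drives lerps with identical inputs, so it has no visible effect.
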Material /Game/Athena/UI/Frontend/ActivityBrowser/Art/M_UI_Game_Mode_Background
graph {
 properties {
  bUsedWithStaticMesh=false
  blendMode=BLEND_Translucent
  materialDomain=MD_UI
  pixelDepthOffsetMode=PDOM_Legacy
  usedShadingModels=Default Lit
 }
 nodes [
  n1 = MaterialFunctionCall() {materialFunction=&GetUserInterfaceUV}
  n2 = MaterialFunctionCall(Float2=@n1.Pixel Size) {materialFunction=&BreakOutFloat2Components}
  n3 = Divide(A=@n2.R, B=Max(A=@n2.G))
  n4 = Constant() {r=1}
  n5 = Multiply(A=Subtract(B=@n3)) {constB=0.5}
  n6 = Constant()
  n7 = Add(A=Multiply(A=AppendVector(A=@n3, B=@n4), B=@n1.Normalized UV), B=Multiply(A=AppendVector(A=@n5, B=@n6)))
  LinearInterpolate(
   A=VectorParameter() {group=Background, parameterName=RGBA-1, defaultValue=<rgba 0 0.0382 0.3325 1>}.RGB
   B=VectorParameter() {group=Background, parameterName=RGBA-2, defaultValue=<rgba 0.0168 0.1144 0.4125 1>}.RGB
   Alpha=MaterialFunctionCall(UVs=@n7, Radius=Multiply(A=@n3) {constB=0.5}, Density=ScalarParameter() {defaultValue=1, parameterName=Density}) {materialFunction=&RadialGradientExponential}.RadialGradientExponential
  )
  n8 = ScalarParameter() {defaultValue=-0.0857, group=TriangleMask, parameterName=S-TriangleRotation, sortPriority=1}
  n9 = Constant() {r=1}
  n10 = MaterialFunctionCall() {materialFunction=&GetUserInterfaceUV}
  n11 = MaterialFunctionCall(Float2=@n10.Pixel Size) {materialFunction=&BreakOutFloat2Components}
  n12 = Divide(A=@n11.G, B=Max(A=@n11.R) {constB=0.1})
  n13 = Multiply(
   A=Multiply(A=AppendVector(A=@n9, B=@n12), B=@n10.Normalized UV)
   B=ScalarParameter() {defaultValue=4, group=TriangleMask, parameterName=Global-Triangle-Mask-Scale}
  )
  n14 = TextureSample(UVs=Multiply(
    A=Rotator(Coordinate=@n13, Time=@n8) {speed=6.2832}
    B=ScalarParameter() {defaultValue=2, group=TriangleMask, parameterName=S-TriangleTiles, sortPriority=1}
   ) {constB=20}) {texture=/Game/Global/Textures/Default/Masks_and_Gradients/T-TriangleGrid}
  n15 = TextureCoordinate()
  n16 = ScalarParameter() {defaultValue=-0.05, group=TriangleMask, parameterName=Global-Triangle-Mask-Rotation}
  n17 = Sine(None=Panner(
    Coordinate=Rotator(Coordinate=@n15, Time=@n16) {speed=6.2832}
    Speed=ScalarParameter() {defaultValue=0.5, group=TriangleMask, parameterName=Global-Triangle-Mask-Speed}
   ) {speedX=1}) {period=4}
  n18 = LinearInterpolate(Alpha=Clamp(None=ComponentMask(None=@n17) {r=true}))
  n19 = LinearInterpolate(A=@n14.R, B=@n14.R, Alpha=@n18)
  n20 = ScalarParameter() {defaultValue=-0.0857, group=TriangleMask, parameterName=L-TriangleRotation, sortPriority=1}
  n21 = TextureSample(UVs=Multiply(
    A=Rotator(Coordinate=@n13, Time=@n20) {speed=6.2832}
    B=ScalarParameter() {defaultValue=1.25, group=TriangleMask, parameterName=L-TriangleTiles, sortPriority=1}
   ) {constB=20}) {texture=&T-TriangleGrid}
  n22 = LinearInterpolate(A=@n21.R, B=@n21.R, Alpha=@n18)
  n23 = Clamp(None=Multiply(A=Add(A=@n19, B=@n22)) {constB=0.5})
  n24 = MaterialFunctionCall(
   UVs=@n7
   Radius=ScalarParameter() {defaultValue=1.5, parameterName=BlurMask}
   Density=ScalarParameter() {defaultValue=2.25, parameterName=BlurMaskdensity}
  ) {materialFunction=&RadialGradientExponential}
  n25 = ScalarParameter() {defaultValue=1, parameterName=Opacity}
  Multiply(
   A=StaticSwitchParameter(True=LinearInterpolate(A=@n23, Alpha=@n24.RadialGradientExponential), False=@n25) {parameterName=Transperant}
   B=@n25
  )
 ]
}
